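# Masked skeletal weapon material for Demon Mask Fire that fades the weapon in and out with a burning edge.
# Custom primitive timestamps drive a timed fade through radial and vertical local-position gradients, combined by Max.
# Downward-scrolling fire noise, projected on two local planes, breaks up the opacity mask and emissive edge.
# These layer over the weapon master and accessory override functions, with optional vertex-color hiding, before the Creative move-tool function.
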Material /DemonMaskFire/Assets/Materials/M_DemonMaskFire_Weapon
graph {
 properties {
  bAutomaticallySetUsageInEditor=false
  bUseMaterialAttributes=true
  bUsedWithClothing=true
  bUsedWithMeshDeformer=true
  bUsedWithMorphTargets=true
  bUsedWithNanite=true
  bUsedWithSkeletalMesh=true
  blendMode=BLEND_Masked
  pixelDepthOffsetMode=PDOM_Legacy
  refractionMethod=RM_IndexOfRefraction
  usedShadingModels=Default Lit
 }
 nodes [
  n1 = GetMaterialAttributes(None=MaterialFunctionCall!(In) {materialFunction=/Game/Packages/Fortress_SharedMaterials/Base_Material_Functions/BaseCustomizationFunctions/MF_FN_Weapon_MASTER}.Material Result) {attributeGetTypes [#1]}
  n2 = MaterialFunctionCall!(WPO) {materialFunction=/Game/Packages/Fortress_SharedMaterials/Base_Material_Functions/Fortnite_HideElementsWithVertexColor}
  n3 = MaterialFunctionCall(In=SetMaterialAttributes!(
    MaterialAttributes=@n1.MaterialAttributes
    World Position Offset=StaticSwitchParameter(True=@n2.Result, False=@n1.World Position Offset) {group=Opacity Mask, parameterName=Use Vertex Colors for Mask}
   ) {attributeSetTypes [#1]}) {materialFunction=&MF_TechArtOverrides_accessories}
  n4 = GetMaterialAttributes(None=@n3.Result) {attributeGetTypes [#2]}
  n5 = VertexInterpolator(VS=LocalPosition() {localOrigin=InstancePreSkinning}.XYZ)
  n6 = OneMinus(None=Distance(A=@n5.PS, B=Constant()))
  n7 = Constant() {r=20}
  n8 = ScalarParameter() {defaultValue=0.3164, parameterName=FadeLerpRadial1}
  n9 = ScalarParameter() {defaultValue=2.5347, parameterName=FadeLerpRadial2}
  n10 = Subtract(A=Time() {period=1}, B=ScalarParameter() {bUseCustomPrimitiveData=true, parameterName=TimeStampIn})
  n11 = ScalarParameter() {defaultValue=1, parameterName=FadeSpeedIn}
  n12 = Divide(
   A=Subtract(A=Time() {period=1}, B=ScalarParameter() {bUseCustomPrimitiveData=true, parameterName=TimeStamp1Out, primitiveDataIndex=1})
   B=ScalarParameter() {defaultValue=0.7, parameterName=FadeSpeedOut}
  )
  n13 = Frac(None=Multiply(A=Saturate(None=@n12)) {constB=1.01})
  n14 = Saturate(None=Min(A=Divide(A=@n10, B=@n11), B=OneMinus(None=@n13)))
  n15 = Min(A=@n14) {constB=0.5}
  n16 = OneMinus(None=ComponentMask(None=@n5.PS) {b=true})
  n17 = Constant() {r=20}
  n18 = ScalarParameter() {defaultValue=-0.5026, parameterName=FadeVertical1}
  n19 = ScalarParameter() {defaultValue=2.2373, parameterName=FadeVertical2}
  n20 = Max(
   A=Add(A=Divide(A=@n6, B=@n7), B=LinearInterpolate(A=@n8, B=@n9, Alpha=@n15) {constA=0.352, constB=0.69}) {constB=0.66}
   B=Add(A=Divide(A=@n16, B=@n17), B=LinearInterpolate(A=@n18, B=@n19, Alpha=@n14) {constA=0.3, constB=0.66}) {constB=0.66}
  )
  n21 = OneMinus(None=Saturate(None=Abs(None=@n20)))
  n22 = Multiply(A=@n21, B=@n21)
  n23 = LocalPosition() {localOrigin=InstancePreSkinning}
  n24 = Multiply(A=Time(), B=Constant3Vector() {constant=<rgba 0 0 -10 0>})
  n25 = Divide(A=VertexInterpolator(VS=Add(A=@n23.XYZ, B=@n24)).PS, B=Constant() {r=70})
  n26 = Max(
   A=TextureSample(UVs=ComponentMask(None=@n25) {b=true, r=true}) {texture=/Game/Global/Textures/Default/Noise/T_FireNoise_01}.R
   B=TextureSample(UVs=ComponentMask(None=@n25) {b=true, g=true}) {texture=&T_FireNoise_01}.R
  )
  n27 = Multiply(A=@n26, B=@n26)
  n28 = OneMinus(None=@n14)
  n29 = Multiply(A=Multiply(A=@n28, B=@n28)) {constB=10}
  n30 = Multiply(
   A=Multiply(A=@n22, B=Add(A=@n22, B=@n27))
   B=Multiply(A=VectorParameter() {parameterName=Emission Color, defaultValue=<rgba 20 2.8583 0.4644 1>}.RGB, B=Add(A=@n29))
  )
  n31 = Saturate(None=@n20)
  n32 = Add(A=@n31, B=@n26)
  MaterialFunctionCall(MA In=SetMaterialAttributes!(MaterialAttributes=@n4.MaterialAttributes, Emissive Color=Add(A=@n4.Emissive Color, B=@n30), Opacity Mask=Multiply(A=@n31, B=@n32)) {attributeSetTypes [#2, #3]}) {materialFunction=&MF_CreativeMoveTool}
 ]
}
